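# Sound cue for the Nug creature's aggro-start vocalization, played as a tactical creature vocalization with prey aggro attenuation.
# A quality-level root routes its first two levels to a random pick from fifteen equally weighted takes, and its last two to a five-take subset.
# The smaller random node is excluded from branch culling.
# Two Nug VO concurrency sets, one global and one priority-setting, govern overlapping playback.
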SoundCue /Irwin/Sounds/Nug/Vocalizations/Nug_AggroStart_Cue
root=QualityLevel_0
nodes [
 {id=QualityLevel_0, children [Random_0,Random_0,Random_1,Random_1]}
 {
  id=Random_0
  children [
   WavePlayer_0
   WavePlayer_1
   WavePlayer_2
   WavePlayer_3
   WavePlayer_4
   WavePlayer_5
   WavePlayer_6
   WavePlayer_7
   WavePlayer_8
   WavePlayer_9
   WavePlayer_10
   WavePlayer_11
   WavePlayer_12
   WavePlayer_13
   WavePlayer_14
  ]
  properties {weights [1,1,1,1,1,1,1,1,1,1,1,1,1,1,1]}
 }
 {
  id=Random_1
  children [WavePlayer_15, WavePlayer_16, WavePlayer_17, WavePlayer_18, WavePlayer_19]
  properties {bShouldExcludeFromBranchCulling=true, weights [1,1,1,1,1]}
 }
 {id=WavePlayer_0, properties {soundWaveAssetPtr=&/Nug_AggroStart_01}}
 {id=WavePlayer_1, properties {soundWaveAssetPtr=&/Nug_AggroStart_02}}
 {id=WavePlayer_10, properties {soundWaveAssetPtr=&/Nug_AggroStart_11}}
 {id=WavePlayer_11, properties {soundWaveAssetPtr=&/Nug_AggroStart_12}}
 {id=WavePlayer_12, properties {soundWaveAssetPtr=&/Nug_AggroStart_13}}
 {id=WavePlayer_13, properties {soundWaveAssetPtr=&/Nug_AggroStart_14}}
 {id=WavePlayer_14, properties {soundWaveAssetPtr=&/Nug_AggroStart_15}}
 {id=WavePlayer_15, properties {soundWaveAssetPtr=&Nug_AggroStart_01}}
 {id=WavePlayer_16, properties {soundWaveAssetPtr=&Nug_AggroStart_02}}
 {id=WavePlayer_17, properties {soundWaveAssetPtr=&/Nug_AggroStart_04}}
 {id=WavePlayer_18, properties {soundWaveAssetPtr=&/Nug_AggroStart_06}}
 {id=WavePlayer_19, properties {soundWaveAssetPtr=&/Nug_AggroStart_07}}
 {id=WavePlayer_2, properties {soundWaveAssetPtr=&/Nug_AggroStart_03}}
 {id=WavePlayer_3, properties {soundWaveAssetPtr=&Nug_AggroStart_04}}
 {id=WavePlayer_4, properties {soundWaveAssetPtr=&/Nug_AggroStart_05}}
 {id=WavePlayer_5, properties {soundWaveAssetPtr=&Nug_AggroStart_06}}
 {id=WavePlayer_6, properties {soundWaveAssetPtr=&Nug_AggroStart_07}}
 {id=WavePlayer_7, properties {soundWaveAssetPtr=&/Nug_AggroStart_08}}
 {id=WavePlayer_8, properties {soundWaveAssetPtr=&/Nug_AggroStart_09}}
 {id=WavePlayer_9, properties {soundWaveAssetPtr=&/Nug_AggroStart_10}}
]
properties {
 attenuationSettings=/Irwin/Sounds/Settings/Prey_Aggro_Vocalization_Att
 maxDistance=7000
 priority=10
 soundClassObject=&Fort_Creatures_Vocalization_Tactical
 volumeMultiplier=1.5
 concurrencySet [/Irwin/Sounds/Settings/Conc/Nug_Global_VO_Conc, /Irwin/Sounds/Settings/Conc/Nug_SetPriority_VO_Conc]
}
