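# Sound cue for the Nug creature's death vocalization, built from Nug startle recordings.
# A quality-level switch sends the first two levels to a random pick of five waves and the last two to a smaller pool of three, which is excluded from branch culling.
# It uses prey vocalization attenuation, the creature vocalization sound class, and two Nug voice concurrency sets for global and priority limiting.
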SoundCue /Irwin/Sounds/Nug/Vocalizations/Nug_Death_Cue
root=QualityLevel_0
nodes [
 {id=QualityLevel_0, children [Random_0,Random_0,Random_1,Random_1]}
 {id=Random_0, children [WavePlayer_1,WavePlayer_2,WavePlayer_3,WavePlayer_4,WavePlayer_5], properties {weights [1,1,1,1,1]}}
 {id=Random_1, children [WavePlayer_0,WavePlayer_6,WavePlayer_7], properties {bShouldExcludeFromBranchCulling=true, weights [1,1,1]}}
 {id=WavePlayer_0, properties {soundWaveAssetPtr=&/Nug_Startle_01}}
 {id=WavePlayer_1, properties {soundWaveAssetPtr=&Nug_Startle_01}}
 {id=WavePlayer_2, properties {soundWaveAssetPtr=&/Nug_Startle_02}}
 {id=WavePlayer_3, properties {soundWaveAssetPtr=&/Nug_Startle_03}}
 {id=WavePlayer_4, properties {soundWaveAssetPtr=&/Nug_Startle_04}}
 {id=WavePlayer_5, properties {soundWaveAssetPtr=&/Nug_Startle_05}}
 {id=WavePlayer_6, properties {soundWaveAssetPtr=&Nug_Startle_02}}
 {id=WavePlayer_7, properties {soundWaveAssetPtr=&Nug_Startle_03}}
]
properties {
 attenuationSettings=/Irwin/Sounds/Settings/Prey_Vocalization_Att
 maxDistance=6350
 soundClassObject=/Game/Sounds/FortSoundClassesAndModes/Fort_Creatures_Vocalization_Normal
 volumeMultiplier=1.3
 concurrencySet [/Irwin/Sounds/Settings/Conc/Nug_Global_VO_Conc, /Irwin/Sounds/Settings/Conc/Nug_SetPriority_VO_Conc]
}
