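# Sound cue for the Nug creature's eat-bait vocalization, using prey vocalization attenuation and the creature vocalization sound class.
# A quality-level node routes its first two levels to one mixer and its last two to another.
# Each mixer layers a random peck with a random, louder eat-bait sound.
# The second mixer draws from fewer waves, has empty slots among its pecks, and excludes its random nodes from branch culling.
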SoundCue /Irwin/Sounds/Nug/Vocalizations/Nug_EatBait_Cue
root=QualityLevel_0
nodes [
 {id=Mixer_0, children [Random_1,Random_0], properties {inputVolume [1,2]}}
 {id=Mixer_1, children [Random_3,Random_2], properties {inputVolume [1,2]}}
 {id=QualityLevel_0, children [Mixer_0,Mixer_0,Mixer_1,Mixer_1]}
 {id=Random_0, children [WavePlayer_0,WavePlayer_1,WavePlayer_2,WavePlayer_3], properties {weights [1,1,1,1]}}
 {
  id=Random_1
  children [WavePlayer_4, WavePlayer_5, WavePlayer_6, WavePlayer_7, WavePlayer_8, WavePlayer_10]
  properties {weights [1,1,1,1,1,1]}
 }
 {id=Random_2, children [WavePlayer_11, WavePlayer_12, WavePlayer_13], properties {bShouldExcludeFromBranchCulling=true, weights [1,1,1]}}
 {
  id=Random_3
  children [WavePlayer_15, null, WavePlayer_17, WavePlayer_18, null, WavePlayer_20, null]
  properties {bShouldExcludeFromBranchCulling=true, weights [1,1,1,1,1,1,1]}
 }
 {id=WavePlayer_0, properties {soundWaveAssetPtr=&/Nug_EatBait_01}}
 {id=WavePlayer_1, properties {soundWaveAssetPtr=&/Nug_EatBait_02}}
 {id=WavePlayer_10, properties {soundWaveAssetPtr=&/Nug_Peck_100}}
 {id=WavePlayer_11, properties {soundWaveAssetPtr=&Nug_EatBait_01}}
 {id=WavePlayer_12, properties {soundWaveAssetPtr=&Nug_EatBait_02}}
 {id=WavePlayer_13, properties {soundWaveAssetPtr=&/Nug_EatBait_03}}
 {id=WavePlayer_15, properties {soundWaveAssetPtr=&/Nug_Peck_110}}
 {id=WavePlayer_17, properties {soundWaveAssetPtr=&/Nug_Peck_060}}
 {id=WavePlayer_18, properties {soundWaveAssetPtr=&/Nug_Peck_070}}
 {id=WavePlayer_2, properties {soundWaveAssetPtr=&Nug_EatBait_03}}
 {id=WavePlayer_20, properties {soundWaveAssetPtr=&/Nug_Peck_090}}
 {id=WavePlayer_3, properties {soundWaveAssetPtr=&/Nug_EatBait_04}}
 {id=WavePlayer_4, properties {soundWaveAssetPtr=&Nug_Peck_110}}
 {id=WavePlayer_5, properties {soundWaveAssetPtr=&/Nug_Peck_050}}
 {id=WavePlayer_6, properties {soundWaveAssetPtr=&Nug_Peck_060}}
 {id=WavePlayer_7, properties {soundWaveAssetPtr=&Nug_Peck_070}}
 {id=WavePlayer_8, properties {soundWaveAssetPtr=&/Nug_Peck_080}}
]
properties {
 attenuationSettings=/Irwin/Sounds/Settings/Prey_Vocalization_Att
 maxDistance=6350
 soundClassObject=/Game/Sounds/FortSoundClassesAndModes/Fort_Creatures_Vocalization_Normal
 volumeMultiplier=1
}
