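# Sound cue for the Nug creature's idle flying vocalizations, using prey vocalization attenuation and Nug-specific voice concurrency.
# A quality-level switch sends the first two levels to a large random pool and the last two to a smaller one that is excluded from branch culling.
# Both pools mostly pick single FlyIdle waves, with low-weight concatenators occasionally playing two waves back to back.
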SoundCue /Irwin/Sounds/Nug/Vocalizations/Nug_FlyIdle_Cue
root=QualityLevel_0
nodes [
 {id=Concatenator_0, children [WavePlayer_18, WavePlayer_13], properties {inputVolume [1,1]}}
 {id=Concatenator_1, children [WavePlayer_2, WavePlayer_12], properties {inputVolume [1,1]}}
 {id=Concatenator_2, children [WavePlayer_17, WavePlayer_20], properties {inputVolume [1,1]}}
 {id=QualityLevel_0, children [Random_0,Random_0,Random_1,Random_1]}
 {
  id=Random_0
  children [
   WavePlayer_0
   WavePlayer_1
   WavePlayer_5
   WavePlayer_6
   WavePlayer_7
   WavePlayer_8
   WavePlayer_9
   WavePlayer_4
   WavePlayer_10
   Concatenator_1
   Concatenator_2
  ]
  properties {weights [1,1,1,1,1,1,1,1,1,0.1,0.1]}
 }
 {
  id=Random_1
  children [WavePlayer_3, WavePlayer_11, WavePlayer_15, Concatenator_0]
  properties {bShouldExcludeFromBranchCulling=true, weights [1,1,1,0.05]}
 }
 {id=WavePlayer_0, properties {soundWaveAssetPtr=&/Nug_FlyIdle_100}}
 {id=WavePlayer_1, properties {soundWaveAssetPtr=&/Nug_FlyIdle_010}}
 {id=WavePlayer_10, properties {soundWaveAssetPtr=&/Nug_FlyIdle_040}}
 {id=WavePlayer_11, properties {soundWaveAssetPtr=&Nug_FlyIdle_010}}
 {id=WavePlayer_12, properties {soundWaveAssetPtr=&/Nug_FlyIdle_140}}
 {id=WavePlayer_13, properties {soundWaveAssetPtr=&/Nug_FlyIdle_170}}
 {id=WavePlayer_15, properties {soundWaveAssetPtr=&/Nug_FlyIdle_080}}
 {id=WavePlayer_17, properties {soundWaveAssetPtr=&/Nug_FlyIdle_160}}
 {id=WavePlayer_18, properties {soundWaveAssetPtr=&Nug_FlyIdle_160}}
 {id=WavePlayer_2, properties {soundWaveAssetPtr=&/Nug_FlyIdle_130}}
 {id=WavePlayer_20, properties {soundWaveAssetPtr=&Nug_FlyIdle_170}}
 {id=WavePlayer_3, properties {soundWaveAssetPtr=&Nug_FlyIdle_100}}
 {id=WavePlayer_4, properties {soundWaveAssetPtr=&/Nug_FlyIdle_030}}
 {id=WavePlayer_5, properties {soundWaveAssetPtr=&/Nug_FlyIdle_050}}
 {id=WavePlayer_6, properties {soundWaveAssetPtr=&/Nug_FlyIdle_060}}
 {id=WavePlayer_7, properties {soundWaveAssetPtr=&/Nug_FlyIdle_070}}
 {id=WavePlayer_8, properties {soundWaveAssetPtr=&Nug_FlyIdle_080}}
 {id=WavePlayer_9, properties {soundWaveAssetPtr=&/Nug_FlyIdle_090}}
]
properties {
 attenuationSettings=/Irwin/Sounds/Settings/Prey_Vocalization_Att
 maxDistance=6350
 priority=0.5
 soundClassObject=/Game/Sounds/FortSoundClassesAndModes/Fort_Creatures_Vocalization_Normal
 volumeMultiplier=1
 concurrencySet [/Irwin/Sounds/Settings/Conc/Nug_SetPriority_SlowFade_VO_Conc, /Irwin/Sounds/Settings/Conc/Nug_SetPriority_VO_Conc]
}
