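# Sound cue for the Nug creature's held-attack vocalization, variant B, built from short clips.
# A root Random node picks one of nine Nug_HeldAttack_Short recordings with equal weighting.
# Playback uses the shared prey vocalization attenuation, the tactical creature vocalization sound class, a dedicated held-attack concurrency set and a reduced volume.
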SoundCue /Irwin/Sounds/Nug/Vocalizations/Nug_HeldAttackB_Cue
root=Random_0
nodes [
 {
  id=Random_0
  children [WavePlayer_9, WavePlayer_10, WavePlayer_11, WavePlayer_12, WavePlayer_13, WavePlayer_14, WavePlayer_15, WavePlayer_16, WavePlayer_17]
  properties {weights [1,1,1,1,1,1,1,1,1]}
 }
 {id=WavePlayer_10, properties {soundWaveAssetPtr=&/Nug_HeldAttack_Short_02}}
 {id=WavePlayer_11, properties {soundWaveAssetPtr=&/Nug_HeldAttack_Short_03}}
 {id=WavePlayer_12, properties {soundWaveAssetPtr=&/Nug_HeldAttack_Short_04}}
 {id=WavePlayer_13, properties {soundWaveAssetPtr=&/Nug_HeldAttack_Short_05}}
 {id=WavePlayer_14, properties {soundWaveAssetPtr=&/Nug_HeldAttack_Short_06}}
 {id=WavePlayer_15, properties {soundWaveAssetPtr=&/Nug_HeldAttack_Short_07}}
 {id=WavePlayer_16, properties {soundWaveAssetPtr=&/Nug_HeldAttack_Short_08}}
 {id=WavePlayer_17, properties {soundWaveAssetPtr=&/Nug_HeldAttack_Short_09}}
 {id=WavePlayer_9, properties {soundWaveAssetPtr=&/Nug_HeldAttack_Short_01}}
]
properties {
 attenuationSettings=/Irwin/Sounds/Settings/Prey_Vocalization_Att
 maxDistance=6350
 priority=5
 soundClassObject=&Fort_Creatures_Vocalization_Tactical
 volumeMultiplier=0.7
 concurrencySet [/Irwin/Sounds/Settings/Conc/Nug_HeldAttack_VO_Conc]
}
