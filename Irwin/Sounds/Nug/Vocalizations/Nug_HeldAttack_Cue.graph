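# Sound cue for the Nug creature's held-attack peck vocalization.
# A Random node picks one of nine equally weighted peck waves and feeds a Modulator that randomizes pitch slightly upward for variety.
# It uses the shared prey vocalization attenuation, a tactical creature vocalization sound class, and a Nug-wide vocalization concurrency set to manage overlapping calls.
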SoundCue /Irwin/Sounds/Nug/Vocalizations/Nug_HeldAttack_Cue
root=Modulator_0
nodes [
 {id=Modulator_0, children [Random_1], properties {pitchMax=1.15, pitchMin=1.05}}
 {
  id=Random_1
  children [WavePlayer_0,WavePlayer_1,WavePlayer_2,WavePlayer_3,WavePlayer_4,WavePlayer_5,WavePlayer_6,WavePlayer_7,WavePlayer_8]
  properties {weights [1,1,1,1,1,1,1,1,1]}
 }
 {id=WavePlayer_0, properties {soundWaveAssetPtr=&/Nug_Held_Attack_Peck_09}}
 {id=WavePlayer_1, properties {soundWaveAssetPtr=&/Nug_Held_Attack_Peck_01}}
 {id=WavePlayer_2, properties {soundWaveAssetPtr=&/Nug_Held_Attack_Peck_02}}
 {id=WavePlayer_3, properties {soundWaveAssetPtr=&/Nug_Held_Attack_Peck_03}}
 {id=WavePlayer_4, properties {soundWaveAssetPtr=&/Nug_Held_Attack_Peck_04}}
 {id=WavePlayer_5, properties {soundWaveAssetPtr=&/Nug_Held_Attack_Peck_05}}
 {id=WavePlayer_6, properties {soundWaveAssetPtr=&/Nug_Held_Attack_Peck_06}}
 {id=WavePlayer_7, properties {soundWaveAssetPtr=&/Nug_Held_Attack_Peck_07}}
 {id=WavePlayer_8, properties {soundWaveAssetPtr=&/Nug_Held_Attack_Peck_08}}
]
properties {
 attenuationSettings=/Irwin/Sounds/Settings/Prey_Vocalization_Att
 maxDistance=6350
 soundClassObject=&Fort_Creatures_Vocalization_Tactical
 volumeMultiplier=1
 concurrencySet [/Irwin/Sounds/Settings/Conc/Nug_Global_VO_Conc]
}
