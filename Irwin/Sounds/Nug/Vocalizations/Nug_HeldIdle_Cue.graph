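# Sound cue for the idle vocalizations of the Nug creature while it is held, using prey attenuation and creature vocalization class.
# Each play mixes a random held-idle voice clip with a random startle-flap foley at half volume and varied pitch.
# A quality-level switch gives the first two levels the full clip pools and the last two smaller subsets.
# Three Nug voice concurrency sets limit overlapping playback.
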SoundCue /Irwin/Sounds/Nug/Vocalizations/Nug_HeldIdle_Cue
root=QualityLevel_0
nodes [
 {id=Mixer_0, children [Random_0,Modulator_0], properties {inputVolume [1,0.5]}}
 {id=Mixer_1, children [Random_3,Modulator_1], properties {inputVolume [1,0.5]}}
 {id=Modulator_0, children [Random_1], properties {pitchMax=1.2, pitchMin=0.9, volumeMin=1}}
 {id=Modulator_1, children [Random_4], properties {pitchMax=1.2, pitchMin=0.9, volumeMin=1}}
 {id=QualityLevel_0, children [Mixer_0,Mixer_0,Mixer_1,Mixer_1]}
 {
  id=Random_0
  children [
   WavePlayer_0
   WavePlayer_1
   WavePlayer_2
   WavePlayer_3
   WavePlayer_4
   WavePlayer_5
   WavePlayer_6
   WavePlayer_7
   WavePlayer_8
   WavePlayer_9
   WavePlayer_10
  ]
  properties {weights [1,1,1,1,1,1,1,1,1,1,1]}
 }
 {id=Random_1, children [WavePlayer_11, WavePlayer_12, WavePlayer_13, WavePlayer_14, WavePlayer_15], properties {weights [1,1,1,1,1]}}
 {
  id=Random_3
  children [WavePlayer_16, WavePlayer_17, WavePlayer_18, WavePlayer_19]
  properties {bShouldExcludeFromBranchCulling=true, weights [1,1,1,1]}
 }
 {id=Random_4, children [WavePlayer_27, WavePlayer_28, WavePlayer_29], properties {bShouldExcludeFromBranchCulling=true, weights [1,1,1]}}
 {id=WavePlayer_0, properties {soundWaveAssetPtr=&/Nug_HeldIdle_100}}
 {id=WavePlayer_1, properties {soundWaveAssetPtr=&/Nug_HeldIdle_110}}
 {id=WavePlayer_10, properties {soundWaveAssetPtr=&/Nug_HeldIdle_090}}
 {id=WavePlayer_11, properties {soundWaveAssetPtr=/Irwin/Sounds/Nug/Foley/Nug_Foley_StartleFlap_010}}
 {id=WavePlayer_12, properties {soundWaveAssetPtr=/Irwin/Sounds/Nug/Foley/Nug_Foley_StartleFlap_020}}
 {id=WavePlayer_13, properties {soundWaveAssetPtr=/Irwin/Sounds/Nug/Foley/Nug_Foley_StartleFlap_030}}
 {id=WavePlayer_14, properties {soundWaveAssetPtr=/Irwin/Sounds/Nug/Foley/Nug_Foley_StartleFlap_040}}
 {id=WavePlayer_15, properties {soundWaveAssetPtr=/Irwin/Sounds/Nug/Foley/Nug_Foley_StartleFlap_050}}
 {id=WavePlayer_16, properties {soundWaveAssetPtr=&Nug_HeldIdle_100}}
 {id=WavePlayer_17, properties {soundWaveAssetPtr=&Nug_HeldIdle_110}}
 {id=WavePlayer_18, properties {soundWaveAssetPtr=&/Nug_HeldIdle_010}}
 {id=WavePlayer_19, properties {soundWaveAssetPtr=&/Nug_HeldIdle_020}}
 {id=WavePlayer_2, properties {soundWaveAssetPtr=&Nug_HeldIdle_010}}
 {id=WavePlayer_27, properties {soundWaveAssetPtr=&Nug_Foley_StartleFlap_010}}
 {id=WavePlayer_28, properties {soundWaveAssetPtr=&Nug_Foley_StartleFlap_020}}
 {id=WavePlayer_29, properties {soundWaveAssetPtr=&Nug_Foley_StartleFlap_030}}
 {id=WavePlayer_3, properties {soundWaveAssetPtr=&Nug_HeldIdle_020}}
 {id=WavePlayer_4, properties {soundWaveAssetPtr=&/Nug_HeldIdle_030}}
 {id=WavePlayer_5, properties {soundWaveAssetPtr=&/Nug_HeldIdle_040}}
 {id=WavePlayer_6, properties {soundWaveAssetPtr=&/Nug_HeldIdle_050}}
 {id=WavePlayer_7, properties {soundWaveAssetPtr=&/Nug_HeldIdle_060}}
 {id=WavePlayer_8, properties {soundWaveAssetPtr=&/Nug_HeldIdle_070}}
 {id=WavePlayer_9, properties {soundWaveAssetPtr=&/Nug_HeldIdle_080}}
]
properties {
 attenuationSettings=/Irwin/Sounds/Settings/Prey_Vocalization_Att
 maxDistance=6350
 priority=0.65
 soundClassObject=/Game/Sounds/FortSoundClassesAndModes/Fort_Creatures_Vocalization_Normal
 concurrencySet [
  /Irwin/Sounds/Settings/Conc/Nug_SetPriority_SlowFade_VO_Conc
  /Irwin/Sounds/Settings/Conc/Nug_Held_VO_Conc
  /Irwin/Sounds/Settings/Conc/Nug_HeldAttack_VO_Conc
 ]
}
