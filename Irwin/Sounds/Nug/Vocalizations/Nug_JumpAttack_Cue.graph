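# Sound cue for the Nug creature's jump attack vocalization.
# A Random node picks one of nine equally weighted long held-attack waves.
# It uses prey vocalization attenuation and the tactical creature vocalization sound class.
# Overridden concurrency allows only one instance at a time, blocking new plays while one is active.
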SoundCue /Irwin/Sounds/Nug/Vocalizations/Nug_JumpAttack_Cue
root=Random_0
nodes [
 {
  id=Random_0
  children [WavePlayer_0,WavePlayer_1,WavePlayer_2,WavePlayer_3,WavePlayer_4,WavePlayer_5,WavePlayer_6,WavePlayer_7,WavePlayer_8]
  properties {weights [1,1,1,1,1,1,1,1,1]}
 }
 {id=WavePlayer_0, properties {soundWaveAssetPtr=&/Nug_HeldAttack_Long_01}}
 {id=WavePlayer_1, properties {soundWaveAssetPtr=&/Nug_HeldAttack_Long_02}}
 {id=WavePlayer_2, properties {soundWaveAssetPtr=&/Nug_HeldAttack_Long_03}}
 {id=WavePlayer_3, properties {soundWaveAssetPtr=&/Nug_HeldAttack_Long_04}}
 {id=WavePlayer_4, properties {soundWaveAssetPtr=&/Nug_HeldAttack_Long_05}}
 {id=WavePlayer_5, properties {soundWaveAssetPtr=&/Nug_HeldAttack_Long_06}}
 {id=WavePlayer_6, properties {soundWaveAssetPtr=&/Nug_HeldAttack_Long_07}}
 {id=WavePlayer_7, properties {soundWaveAssetPtr=&/Nug_HeldAttack_Long_08}}
 {id=WavePlayer_8, properties {soundWaveAssetPtr=&/Nug_HeldAttack_Long_09}}
]
properties {
 attenuationSettings=/Irwin/Sounds/Settings/Prey_Vocalization_Att
 bOverrideConcurrency=true
 maxDistance=6350
 soundClassObject=&Fort_Creatures_Vocalization_Tactical
 concurrencyOverrides {maxCount=1, resolutionRule=PreventNew}
}
