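# Sound cue for the Nug creature's jump start, layering a vocalization with a delayed startle-flap foley sound under prey vocalization attenuation.
# Each layer picks a random wave and modulates its pitch, with a fixed short delay on the flap, before a mixer weights the two.
# A quality-level node feeds the first two tiers the full variation pool and the last two a smaller pool with different mix volumes.
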SoundCue /Irwin/Sounds/Nug/Vocalizations/Nug_JumpStart_Cue
root=QualityLevel_0
nodes [
 {id=Delay_0, children [Random_4], properties {delayMax=0.15, delayMin=0.15}}
 {id=Delay_1, children [Random_2], properties {delayMax=0.15, delayMin=0.15}}
 {id=Mixer_0, children [Modulator_0,Modulator_1], properties {inputVolume [0.5,2]}}
 {id=Mixer_1, children [Modulator_2,Modulator_3], properties {inputVolume [1,1.5]}}
 {id=Modulator_0, children [Random_5], properties {pitchMax=1.1}}
 {id=Modulator_1, children [Delay_0], properties {pitchMax=1.2}}
 {id=Modulator_2, children [Random_0], properties {pitchMax=1.2}}
 {id=Modulator_3, children [Delay_1], properties {pitchMax=1.2}}
 {id=QualityLevel_0, children [Mixer_0,Mixer_0,Mixer_1,Mixer_1]}
 {id=Random_0, children [WavePlayer_25, WavePlayer_26, WavePlayer_27], properties {bShouldExcludeFromBranchCulling=true, weights [1,1,1]}}
 {id=Random_2, children [WavePlayer_12, WavePlayer_13], properties {bShouldExcludeFromBranchCulling=true, weights [1,1]}}
 {id=Random_4, children [WavePlayer_19, WavePlayer_20, WavePlayer_21, WavePlayer_22, WavePlayer_23], properties {weights [1,1,1,1,1]}}
 {
  id=Random_5
  children [WavePlayer_14, WavePlayer_15, WavePlayer_16, WavePlayer_17, WavePlayer_18, WavePlayer_24]
  properties {weights [1,1,1,1,1,1]}
 }
 {id=WavePlayer_12, properties {soundWaveAssetPtr=/Irwin/Sounds/Nug/Foley/Nug_Foley_StartleFlap_010}}
 {id=WavePlayer_13, properties {soundWaveAssetPtr=/Irwin/Sounds/Nug/Foley/Nug_Foley_StartleFlap_020}}
 {id=WavePlayer_14, properties {soundWaveAssetPtr=&/Nug_Jump_Start_01}}
 {id=WavePlayer_15, properties {soundWaveAssetPtr=&/Nug_Jump_Start_02}}
 {id=WavePlayer_16, properties {soundWaveAssetPtr=&/Nug_Jump_Start_03}}
 {id=WavePlayer_17, properties {soundWaveAssetPtr=&/Nug_Jump_Start_04}}
 {id=WavePlayer_18, properties {soundWaveAssetPtr=&/Nug_Jump_Start_05}}
 {id=WavePlayer_19, properties {soundWaveAssetPtr=&Nug_Foley_StartleFlap_010}}
 {id=WavePlayer_20, properties {soundWaveAssetPtr=&Nug_Foley_StartleFlap_020}}
 {id=WavePlayer_21, properties {soundWaveAssetPtr=/Irwin/Sounds/Nug/Foley/Nug_Foley_StartleFlap_030}}
 {id=WavePlayer_22, properties {soundWaveAssetPtr=/Irwin/Sounds/Nug/Foley/Nug_Foley_StartleFlap_040}}
 {id=WavePlayer_23, properties {soundWaveAssetPtr=/Irwin/Sounds/Nug/Foley/Nug_Foley_StartleFlap_050}}
 {id=WavePlayer_24, properties {soundWaveAssetPtr=&/Nug_Jump_Start_06}}
 {id=WavePlayer_25, properties {soundWaveAssetPtr=&Nug_Jump_Start_01}}
 {id=WavePlayer_26, properties {soundWaveAssetPtr=&Nug_Jump_Start_02}}
 {id=WavePlayer_27, properties {soundWaveAssetPtr=&Nug_Jump_Start_03}}
]
properties {
 attenuationSettings=/Irwin/Sounds/Settings/Prey_Vocalization_Att
 maxDistance=6350
 soundClassObject=&Fort_Creatures_Vocalization_Tactical
 volumeMultiplier=1.5
 concurrencySet [/Irwin/Sounds/Settings/Conc/Nug_Global_VO_Conc, /Irwin/Sounds/Settings/Conc/Nug_SetPriority_VO_Conc]
}
