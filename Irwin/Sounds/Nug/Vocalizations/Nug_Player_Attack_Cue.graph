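# Sound cue for the Nug creature's player-attack vocalization, using prey vocalization attenuation and the creature vocalization sound class.
# A quality-level switch sends its first two slots to an equally weighted random pool of sixteen takes and its last two to a six-take subset.
# A modulator at the root varies the chosen take.
# Two Nug voice concurrency settings, one global and one named for priority, manage overlapping playback.
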SoundCue /Irwin/Sounds/Nug/Vocalizations/Nug_Player_Attack_Cue
root=Modulator_0
nodes [
 {id=Modulator_0, children [QualityLevel_0]}
 {id=QualityLevel_0, children [Random_0,Random_0,Random_1,Random_1]}
 {
  id=Random_0
  children [
   WavePlayer_0
   WavePlayer_1
   WavePlayer_2
   WavePlayer_3
   WavePlayer_4
   WavePlayer_5
   WavePlayer_6
   WavePlayer_7
   WavePlayer_8
   WavePlayer_9
   WavePlayer_10
   WavePlayer_11
   WavePlayer_12
   WavePlayer_13
   WavePlayer_14
   WavePlayer_15
  ]
  properties {weights [1,1,1,1,1,1,1,1,1,1,1,1,1,1,1,1]}
 }
 {
  id=Random_1
  children [WavePlayer_16, WavePlayer_17, WavePlayer_18, WavePlayer_19, WavePlayer_20, WavePlayer_21]
  properties {bShouldExcludeFromBranchCulling=true, weights [1,1,1,1,1,1]}
 }
 {id=WavePlayer_0, properties {soundWaveAssetPtr=&/Nug_Player_Attack_16}}
 {id=WavePlayer_1, properties {soundWaveAssetPtr=&/Nug_Player_Attack_01}}
 {id=WavePlayer_10, properties {soundWaveAssetPtr=&/Nug_Player_Attack_10}}
 {id=WavePlayer_11, properties {soundWaveAssetPtr=&/Nug_Player_Attack_11}}
 {id=WavePlayer_12, properties {soundWaveAssetPtr=&/Nug_Player_Attack_12}}
 {id=WavePlayer_13, properties {soundWaveAssetPtr=&/Nug_Player_Attack_13}}
 {id=WavePlayer_14, properties {soundWaveAssetPtr=&/Nug_Player_Attack_14}}
 {id=WavePlayer_15, properties {soundWaveAssetPtr=&/Nug_Player_Attack_15}}
 {id=WavePlayer_16, properties {soundWaveAssetPtr=&Nug_Player_Attack_16}}
 {id=WavePlayer_17, properties {soundWaveAssetPtr=&Nug_Player_Attack_01}}
 {id=WavePlayer_18, properties {soundWaveAssetPtr=&/Nug_Player_Attack_02}}
 {id=WavePlayer_19, properties {soundWaveAssetPtr=&/Nug_Player_Attack_03}}
 {id=WavePlayer_2, properties {soundWaveAssetPtr=&Nug_Player_Attack_02}}
 {id=WavePlayer_20, properties {soundWaveAssetPtr=&/Nug_Player_Attack_04}}
 {id=WavePlayer_21, properties {soundWaveAssetPtr=&/Nug_Player_Attack_05}}
 {id=WavePlayer_3, properties {soundWaveAssetPtr=&Nug_Player_Attack_03}}
 {id=WavePlayer_4, properties {soundWaveAssetPtr=&Nug_Player_Attack_04}}
 {id=WavePlayer_5, properties {soundWaveAssetPtr=&Nug_Player_Attack_05}}
 {id=WavePlayer_6, properties {soundWaveAssetPtr=&/Nug_Player_Attack_06}}
 {id=WavePlayer_7, properties {soundWaveAssetPtr=&/Nug_Player_Attack_07}}
 {id=WavePlayer_8, properties {soundWaveAssetPtr=&/Nug_Player_Attack_08}}
 {id=WavePlayer_9, properties {soundWaveAssetPtr=&/Nug_Player_Attack_09}}
]
properties {
 attenuationSettings=/Irwin/Sounds/Settings/Prey_Vocalization_Att
 maxDistance=6350
 priority=10
 soundClassObject=&Fort_Creatures_Vocalization_Tactical
 volumeMultiplier=1
 concurrencySet [/Irwin/Sounds/Settings/Conc/Nug_Global_VO_Conc, /Irwin/Sounds/Settings/Conc/Nug_SetPriority_VO_Conc]
}
