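# Sound cue for the Nug creature's small cluck vocalizations, using prey attenuation, the creature vocalization sound class and Nug voice concurrency limits.
# Its root randomly picks either a cluck or an empty branch, so some triggers stay silent.
# The cluck branch varies pitch and volume through a modulator, then a quality switch selects an eleven-wave random pool or a five-wave subset.
# Despite the Loop name, the cue contains no looping node.
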SoundCue /Irwin/Sounds/Nug/Vocalizations/Nug_SmallCluckLoop_Cue
root=Random_2
nodes [
 {id=Modulator_0, children [QualityLevel_0], properties {pitchMax=1.3, pitchMin=0.9, volumeMin=0.5}}
 {id=QualityLevel_0, children [Random_0,Random_0,Random_1,Random_1]}
 {
  id=Random_0
  children [
   WavePlayer_1
   WavePlayer_2
   WavePlayer_3
   WavePlayer_4
   WavePlayer_5
   WavePlayer_6
   WavePlayer_0
   WavePlayer_8
   WavePlayer_9
   WavePlayer_10
   WavePlayer_11
  ]
  properties {weights [1,1,1,1,1,1,1,1,1,1,1]}
 }
 {
  id=Random_1
  children [WavePlayer_7, WavePlayer_12, WavePlayer_16, WavePlayer_21, WavePlayer_17]
  properties {bShouldExcludeFromBranchCulling=true, weights [1,1,1,1,1]}
 }
 {id=Random_2, children [Modulator_0,null], properties {bShouldExcludeFromBranchCulling=true, weights [1,0.7]}}
 {id=WavePlayer_0, properties {soundWaveAssetPtr=&/Nug_SmallCluck_06a}}
 {id=WavePlayer_1, properties {soundWaveAssetPtr=&/Nug_SmallCluck_01}}
 {id=WavePlayer_10, properties {soundWaveAssetPtr=&/Nug_SmallCluck_08}}
 {id=WavePlayer_11, properties {soundWaveAssetPtr=&/Nug_SmallCluck_09}}
 {id=WavePlayer_12, properties {soundWaveAssetPtr=&/Nug_SmallCluck_010}}
 {id=WavePlayer_16, properties {soundWaveAssetPtr=&/Nug_SmallCluck_04}}
 {id=WavePlayer_17, properties {soundWaveAssetPtr=&/Nug_SmallCluck_06}}
 {id=WavePlayer_2, properties {soundWaveAssetPtr=&Nug_SmallCluck_010}}
 {id=WavePlayer_21, properties {soundWaveAssetPtr=&Nug_SmallCluck_06a}}
 {id=WavePlayer_3, properties {soundWaveAssetPtr=&/Nug_SmallCluck_011}}
 {id=WavePlayer_4, properties {soundWaveAssetPtr=&/Nug_SmallCluck_02}}
 {id=WavePlayer_5, properties {soundWaveAssetPtr=&/Nug_SmallCluck_03}}
 {id=WavePlayer_6, properties {soundWaveAssetPtr=&Nug_SmallCluck_04}}
 {id=WavePlayer_7, properties {soundWaveAssetPtr=&Nug_SmallCluck_01}}
 {id=WavePlayer_8, properties {soundWaveAssetPtr=&Nug_SmallCluck_06}}
 {id=WavePlayer_9, properties {soundWaveAssetPtr=&/Nug_SmallCluck_07}}
]
properties {
 attenuationSettings=/Irwin/Sounds/Settings/Prey_Vocalization_Att
 maxDistance=6350
 priority=0.5
 soundClassObject=/Game/Sounds/FortSoundClassesAndModes/Fort_Creatures_Vocalization_Normal
 volumeMultiplier=1.5
 concurrencySet [/Irwin/Sounds/Settings/Conc/Nug_Global_VO_Conc, /Irwin/Sounds/Settings/Conc/Nug_SetPriority_VO_Conc]
}
