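# Sound cue for the Nug creature's "Sneezed Buttercake" vocalization, routed to the tactical creature vocalization sound class with prey sneeze attenuation.
# A quality-level node feeds two equal-weight random pools: eight variants for the first two levels, the first four of those for the other two.
# The smaller pool is excluded from branch culling.
# A concurrency override allows one instance per owner, stopping the lowest priority.
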SoundCue /Irwin/Sounds/Nug/Vocalizations/Nug_Sneezed_Buttercake_Cue
root=QualityLevel_0
nodes [
 {id=QualityLevel_0, children [Random_0,Random_0,Random_1,Random_1]}
 {
  id=Random_0
  children [WavePlayer_0,WavePlayer_1,WavePlayer_2,WavePlayer_3,WavePlayer_4,WavePlayer_5,WavePlayer_6,WavePlayer_7]
  properties {weights [1,1,1,1,1,1,1,1]}
 }
 {
  id=Random_1
  children [WavePlayer_8, WavePlayer_9, WavePlayer_10, WavePlayer_11]
  properties {bShouldExcludeFromBranchCulling=true, weights [1,1,1,1]}
 }
 {id=WavePlayer_0, properties {soundWaveAssetPtr=&/Nug_Sneezed_Buttercake_01}}
 {id=WavePlayer_1, properties {soundWaveAssetPtr=&/Nug_Sneezed_Buttercake_02}}
 {id=WavePlayer_10, properties {soundWaveAssetPtr=&/Nug_Sneezed_Buttercake_03}}
 {id=WavePlayer_11, properties {soundWaveAssetPtr=&/Nug_Sneezed_Buttercake_04}}
 {id=WavePlayer_2, properties {soundWaveAssetPtr=&Nug_Sneezed_Buttercake_03}}
 {id=WavePlayer_3, properties {soundWaveAssetPtr=&Nug_Sneezed_Buttercake_04}}
 {id=WavePlayer_4, properties {soundWaveAssetPtr=&/Nug_Sneezed_Buttercake_05}}
 {id=WavePlayer_5, properties {soundWaveAssetPtr=&/Nug_Sneezed_Buttercake_06}}
 {id=WavePlayer_6, properties {soundWaveAssetPtr=&/Nug_Sneezed_Buttercake_07}}
 {id=WavePlayer_7, properties {soundWaveAssetPtr=&/Nug_Sneezed_Buttercake_08}}
 {id=WavePlayer_8, properties {soundWaveAssetPtr=&Nug_Sneezed_Buttercake_01}}
 {id=WavePlayer_9, properties {soundWaveAssetPtr=&Nug_Sneezed_Buttercake_02}}
]
properties {
 attenuationSettings=/Irwin/Sounds/Settings/Prey_Sneezed_Vocalization_Att
 bOverrideConcurrency=true
 maxDistance=9000
 priority=100
 soundClassObject=&Fort_Creatures_Vocalization_Tactical
 volumeMultiplier=1.4
 concurrencyOverrides {bLimitToOwner=true, maxCount=1, resolutionRule=StopLowestPriority}
}
