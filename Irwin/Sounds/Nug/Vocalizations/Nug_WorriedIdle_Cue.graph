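# Sound cue for the Nug creature's worried idle vocalization, played as a creature vocalization with prey attenuation and Nug voice concurrency limits.
# A Quality Level node sends its first two slots to an equal-weight random pool of eleven takes and its last two to a four-take subset of those waves.
# The smaller pool is excluded from branch culling.
# A root Modulator varies each playback, with a raised pitch ceiling.
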SoundCue /Irwin/Sounds/Nug/Vocalizations/Nug_WorriedIdle_Cue
root=Modulator_0
nodes [
 {id=Modulator_0, children [QualityLevel_0], properties {pitchMax=1.2}}
 {id=QualityLevel_0, children [Random_0,Random_0,Random_1,Random_1]}
 {
  id=Random_0
  children [
   WavePlayer_0
   WavePlayer_1
   WavePlayer_2
   WavePlayer_4
   WavePlayer_5
   WavePlayer_6
   WavePlayer_7
   WavePlayer_8
   WavePlayer_9
   WavePlayer_10
   WavePlayer_11
  ]
  properties {weights [1,1,1,1,1,1,1,1,1,1,1]}
 }
 {
  id=Random_1
  children [WavePlayer_3, WavePlayer_12, WavePlayer_13, WavePlayer_14]
  properties {bShouldExcludeFromBranchCulling=true, weights [1,1,1,1]}
 }
 {id=WavePlayer_0, properties {soundWaveAssetPtr=&/Nug_WorriedIdle_100}}
 {id=WavePlayer_1, properties {soundWaveAssetPtr=&/Nug_WorriedIdle_110}}
 {id=WavePlayer_10, properties {soundWaveAssetPtr=&/Nug_WorriedIdle_070}}
 {id=WavePlayer_11, properties {soundWaveAssetPtr=&/Nug_WorriedIdle_080}}
 {id=WavePlayer_12, properties {soundWaveAssetPtr=&Nug_WorriedIdle_110}}
 {id=WavePlayer_13, properties {soundWaveAssetPtr=&/Nug_WorriedIdle_120}}
 {id=WavePlayer_14, properties {soundWaveAssetPtr=&/Nug_WorriedIdle_010}}
 {id=WavePlayer_2, properties {soundWaveAssetPtr=&Nug_WorriedIdle_120}}
 {id=WavePlayer_3, properties {soundWaveAssetPtr=&Nug_WorriedIdle_100}}
 {id=WavePlayer_4, properties {soundWaveAssetPtr=&Nug_WorriedIdle_010}}
 {id=WavePlayer_5, properties {soundWaveAssetPtr=&/Nug_WorriedIdle_020}}
 {id=WavePlayer_6, properties {soundWaveAssetPtr=&/Nug_WorriedIdle_030}}
 {id=WavePlayer_7, properties {soundWaveAssetPtr=&/Nug_WorriedIdle_040}}
 {id=WavePlayer_8, properties {soundWaveAssetPtr=&/Nug_WorriedIdle_050}}
 {id=WavePlayer_9, properties {soundWaveAssetPtr=&/Nug_WorriedIdle_060}}
]
properties {
 attenuationSettings=/Irwin/Sounds/Settings/Prey_Vocalization_Att
 maxDistance=6350
 priority=4
 soundClassObject=&Fort_Creatures_Vocalization_Tactical
 volumeMultiplier=1.5
 concurrencySet [/Irwin/Sounds/Settings/Conc/Nug_Global_VO_Conc, /Irwin/Sounds/Settings/Conc/Nug_SetPriority_VO_Conc]
}
